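; Gameplay cue notify that plays the burst effect for a thrown Ancient Weapon hitting the environment.
; After the parent burst logic runs, it sets the first particle component's HitPlayer parameter to false and GlobalScale to a reduced size.
; On the AllowJanus branch, it also spawns an FN_RadialForce physics field actor at the cue location, provided the physics field console variable is enabled.
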(fn OnBurst(EitherVFX-preferthisversion)
  (params (MyTarget :type Actor%20Object%20Reference) (Parameters :type Gameplay%20Cue%20Parameters%20Structure%20%28by%20ref%29) (ParticleComponents :type Array%20of%20FXSystem%20Component%20Object%20References) (AudioComponents :type Array%20of%20Audio%20Component%20Object%20References) (BurstCameraShakeInstance :type Camera%20Shake%20Base%20Object%20Reference) (BurstDecalInstance :type Decal%20Actor%20Object%20Reference))
  (bind (_normalizedmagnitude _rawmagnitude _effectcontext _matchedtagname _originaltag _aggregatedsourcetags _aggregatedtargettags _location) (Ability|GameplayCue|BreakGameplayCueParameters Parameters))
  (bind _output (Utilities|Array|Get(acopy) ParticleComponents))
  (|CallParentFunction MyTarget Parameters ParticleComponents AudioComponents BurstCameraShakeInstance BurstDecalInstance)
  (Effects|Components|ParticleSystem|SetBoolParameter _output "HitPlayer" false)
  (Effects|Components|ParticleSystem|SetFloatParameter _output "GlobalScale" 0.25)
  ;[1 1 1] Add RadialForce
  (Utilities|AllowJanus
    (:Allow
      (bind _returnvalue (Development|GetConsoleVariableIntValue "r.PhysicsField.EnableField"))
      (if (== _returnvalue 1)
        (Game|SpawnActorfromClass 0 "/Game/Effects/Niagara/Blueprints/PhsyicsFields/FN_RadialForce.FN_RadialForce_C" (Math|Transform|MakeTransform _location) "Undefined" "MultiplyWithRoot" 0 0.2 0.75 7e+02 0.75 0.0 0)))
    (:Nope
      _))
  (return))
